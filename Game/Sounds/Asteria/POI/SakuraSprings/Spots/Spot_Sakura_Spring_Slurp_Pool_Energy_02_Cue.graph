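# Looping outdoor ambient spot cue for slurp pool energy at the Sakura Springs POI.
# A quality-level switch sends the first two levels to a fixed-pitch modulator and the last two to a random-pitch one, each feeding a looping LQ-named wave.
# A root oscillator slowly swells the volume.
# Looping-spot attenuation with overrides spatializes it, and concurrency allows two instances, stopping the farthest first.
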SoundCue /Game/Sounds/Asteria/POI/SakuraSprings/Spots/Spot_Sakura_Spring_Slurp_Pool_Energy_02_Cue
root=Oscillator_0
nodes [
 {id=Modulator_0, children [WavePlayer_2], properties {pitchMax=1, pitchMin=1}}
 {id=Modulator_1, children [WavePlayer_1], properties {pitchMax=1.2}}
 {
  id=Oscillator_0
  children [QualityLevel_0]
  properties {amplitudeMax=0.4, amplitudeMin=0.3, bModulateVolume=true, centerMax=1, centerMin=1, frequencyMax=0.05, frequencyMin=0.03}
 }
 {id=QualityLevel_0, children [Modulator_0,Modulator_0,Modulator_1,Modulator_1]}
 {
  id=WavePlayer_1
  properties {bLooping=true, soundWaveAssetPtr=&/Spot_Sakura_Springs_Slurp_Pool_Energy_2_LQ}
 }
 {
  id=WavePlayer_2
  properties {bLooping=true, soundWaveAssetPtr=&Spot_Sakura_Springs_Slurp_Pool_Energy_2_LQ}
 }
]
properties {
 attenuationSettings=/Game/Sounds/Attenuation/Fort_Looping_Spot_Att
 bOverrideAttenuation=true
 bOverrideConcurrency=true
 maxDistance=4100
 soundClassObject=&Fort_Ambient_Outdoors_Athena
 volumeMultiplier=0.175
 attenuationOverrides {stereoSpread=4000, attenuationShapeExtents {x=500}}
 concurrencyOverrides {maxCount=2, resolutionRule=StopFarthestThenPreventNew}
}
