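# Sound cue for one-shot bamboo wind chime ambience at the Sakura Springs POI in Asteria.
# A Quality Level root sends its first two levels to a random pick from six chime waves and its last two to a pick from three.
# Each pool passes through its own Modulator for pitch and volume variation.
# Output uses the ambient one-shot sound class, distant-creak attenuation and a lowered volume.
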SoundCue /Game/Sounds/Asteria/POI/SakuraSprings/Spots/Spot_Sakura_Springs_Bamboo_Chime_Cue
root=QualityLevel_0
nodes [
 {id=Modulator_1, children [Random_3], properties {pitchMax=1.1, pitchMin=0.8, volumeMax=1.1, volumeMin=0.8}}
 {id=Modulator_2, children [Random_4], properties {pitchMax=1.1, volumeMax=1.1, volumeMin=0.9}}
 {id=QualityLevel_0, children [Modulator_1,Modulator_1,Modulator_2,Modulator_2]}
 {
  id=Random_3
  children [WavePlayer_16, WavePlayer_17, WavePlayer_18, WavePlayer_19, WavePlayer_20, WavePlayer_21]
  properties {weights [1,1,1,1,1,1]}
 }
 {id=Random_4, children [WavePlayer_22, WavePlayer_23, WavePlayer_24], properties {weights [1,1,1]}}
 {id=WavePlayer_16, properties {soundWaveAssetPtr=&/Spot_Bamboo_Windchime_6}}
 {id=WavePlayer_17, properties {soundWaveAssetPtr=&/Spot_Bamboo_Windchime_1}}
 {id=WavePlayer_18, properties {soundWaveAssetPtr=&/Spot_Bamboo_Windchime_2}}
 {id=WavePlayer_19, properties {soundWaveAssetPtr=&/Spot_Bamboo_Windchime_3}}
 {id=WavePlayer_20, properties {soundWaveAssetPtr=&/Spot_Bamboo_Windchime_4}}
 {id=WavePlayer_21, properties {soundWaveAssetPtr=&/Spot_Bamboo_Windchime_5}}
 {id=WavePlayer_22, properties {soundWaveAssetPtr=&Spot_Bamboo_Windchime_3}}
 {id=WavePlayer_23, properties {soundWaveAssetPtr=&Spot_Bamboo_Windchime_1}}
 {id=WavePlayer_24, properties {soundWaveAssetPtr=&Spot_Bamboo_Windchime_2}}
]
properties {
 attenuationSettings=/Game/Sounds/Attenuation/Fort_TileAmbient_OneShot_Distant_Creak_Att
 bEnableBusSends=false
 maxDistance=8200
 soundClassObject=&Fort_Ambient_OneShots
 volumeMultiplier=0.251
 preEffectBusSends [{}]
}
