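# Sound cue for a crane bird spot ambience at the Sakura Springs POI, using outdoor ambient class and one-shot bird attenuation.
# A mixer layers three random picks from the same seven crane calls: one immediate, one after a short delay, and a quieter one after a longer delay.
# The last layer gets extra random volume that can drop to silence, and a root modulator varies pitch and volume of the whole burst.
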SoundCue /Game/Sounds/Asteria/POI/SakuraSprings/Spots/Spot_Sakura_Springs_Crane_Bird_Cue
root=Modulator_1
nodes [
 {id=Delay_0, children [Random_1], properties {delayMax=0.8, delayMin=0.4}}
 {id=Delay_1, children [Modulator_2], properties {delayMax=1.3, delayMin=0.9}}
 {id=Mixer_2, children [Delay_0,Random_0,Delay_1], properties {inputVolume [1,1,0.3]}}
 {id=Modulator_1, children [Mixer_2], properties {pitchMax=1.1, pitchMin=0.9, volumeMax=1.2, volumeMin=0.8}}
 {id=Modulator_2, children [Random_2], properties {volumeMax=0.8, volumeMin=0}}
 {
  id=Random_0
  children [WavePlayer_8, WavePlayer_10, WavePlayer_11, WavePlayer_12, WavePlayer_13, WavePlayer_14, WavePlayer_24]
  properties {weights [1,1,1,1,1,1,1]}
 }
 {
  id=Random_1
  children [WavePlayer_25, WavePlayer_26, WavePlayer_27, WavePlayer_28, WavePlayer_29, WavePlayer_30, WavePlayer_31]
  properties {weights [1,1,1,1,1,1,1]}
 }
 {
  id=Random_2
  children [WavePlayer_32, WavePlayer_33, WavePlayer_34, WavePlayer_35, WavePlayer_36, WavePlayer_37, WavePlayer_38]
  properties {weights [1,1,1,1,1,1,1]}
 }
 {id=WavePlayer_10, properties {soundWaveAssetPtr=&/Spot_Bird_Crane_7}}
 {id=WavePlayer_11, properties {soundWaveAssetPtr=&/Spot_Bird_Crane_1}}
 {id=WavePlayer_12, properties {soundWaveAssetPtr=&/Spot_Bird_Crane_2}}
 {id=WavePlayer_13, properties {soundWaveAssetPtr=&/Spot_Bird_Crane_3}}
 {id=WavePlayer_14, properties {soundWaveAssetPtr=&/Spot_Bird_Crane_5}}
 {id=WavePlayer_24, properties {soundWaveAssetPtr=&/Spot_Bird_Crane_6}}
 {id=WavePlayer_25, properties {soundWaveAssetPtr=&/Spot_Bird_Crane_4}}
 {id=WavePlayer_26, properties {soundWaveAssetPtr=&Spot_Bird_Crane_7}}
 {id=WavePlayer_27, properties {soundWaveAssetPtr=&Spot_Bird_Crane_1}}
 {id=WavePlayer_28, properties {soundWaveAssetPtr=&Spot_Bird_Crane_2}}
 {id=WavePlayer_29, properties {soundWaveAssetPtr=&Spot_Bird_Crane_3}}
 {id=WavePlayer_30, properties {soundWaveAssetPtr=&Spot_Bird_Crane_5}}
 {id=WavePlayer_31, properties {soundWaveAssetPtr=&Spot_Bird_Crane_6}}
 {id=WavePlayer_32, properties {soundWaveAssetPtr=&Spot_Bird_Crane_4}}
 {id=WavePlayer_33, properties {soundWaveAssetPtr=&Spot_Bird_Crane_7}}
 {id=WavePlayer_34, properties {soundWaveAssetPtr=&Spot_Bird_Crane_1}}
 {id=WavePlayer_35, properties {soundWaveAssetPtr=&Spot_Bird_Crane_2}}
 {id=WavePlayer_36, properties {soundWaveAssetPtr=&Spot_Bird_Crane_3}}
 {id=WavePlayer_37, properties {soundWaveAssetPtr=&Spot_Bird_Crane_5}}
 {id=WavePlayer_38, properties {soundWaveAssetPtr=&Spot_Bird_Crane_6}}
 {id=WavePlayer_8, properties {soundWaveAssetPtr=&Spot_Bird_Crane_4}}
]
properties {
 attenuationSettings=&Fort_TileAmbient_OneShot_Bird_Att
 maxDistance=8005
 soundClassObject=&Fort_Ambient_Outdoors_Athena
 volumeMultiplier=0.226
 preEffectBusSends [{}]
}
