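# Ambient spot Sound Cue that plays intermittent bamboo tree creaks at the Sakura Springs POI.
# Each branch loops a randomized delay, pitch and volume modulation, and an equal-weight random pick of creak waves.
# A quality-level root sends the first two levels to a nine-wave branch and the last two to a three-wave branch.
# It uses outdoor ambient attenuation and a concurrency override limiting it to one voice.
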SoundCue /Game/Sounds/Asteria/POI/SakuraSprings/Spots/Spot_Sakura_Springs_Tree_Bamboo_Creak_Cue
root=QualityLevel_0
nodes [
 {id=Delay_0, children [Random_0], properties {delayMax=12, delayMin=2}}
 {id=Delay_1, children [Random_1], properties {delayMax=12, delayMin=2}}
 {id=Looping_0, children [Modulator_0]}
 {id=Looping_1, children [Modulator_1]}
 {id=Modulator_0, children [Delay_0], properties {pitchMin=0.7}}
 {id=Modulator_1, children [Delay_1], properties {pitchMin=0.7}}
 {id=QualityLevel_0, children [Looping_0,Looping_0,Looping_1,Looping_1]}
 {
  id=Random_0
  children [WavePlayer_0,WavePlayer_1,WavePlayer_2,WavePlayer_4,WavePlayer_5,WavePlayer_6,WavePlayer_7,WavePlayer_8,WavePlayer_9]
  properties {weights [1,1,1,1,1,1,1,1,1]}
 }
 {id=Random_1, children [WavePlayer_3, WavePlayer_10, WavePlayer_11], properties {weights [1,1,1]}}
 {id=WavePlayer_0, properties {soundWaveAssetPtr=&/Spot_Bamboo_Creak_1}}
 {id=WavePlayer_1, properties {soundWaveAssetPtr=&/Spot_Bamboo_Creak_2}}
 {id=WavePlayer_10, properties {soundWaveAssetPtr=&Spot_Bamboo_Creak_2}}
 {id=WavePlayer_11, properties {soundWaveAssetPtr=&/Spot_Bamboo_Creak_3}}
 {id=WavePlayer_2, properties {soundWaveAssetPtr=&Spot_Bamboo_Creak_3}}
 {id=WavePlayer_3, properties {soundWaveAssetPtr=&Spot_Bamboo_Creak_1}}
 {id=WavePlayer_4, properties {soundWaveAssetPtr=&/Spot_Bamboo_Creak_4}}
 {id=WavePlayer_5, properties {soundWaveAssetPtr=&/Spot_Bamboo_Creak_5}}
 {id=WavePlayer_6, properties {soundWaveAssetPtr=&/Spot_Bamboo_Creak_6}}
 {id=WavePlayer_7, properties {soundWaveAssetPtr=&/Spot_Bamboo_Creak_7}}
 {id=WavePlayer_8, properties {soundWaveAssetPtr=&/Spot_Bamboo_Creak_8}}
 {id=WavePlayer_9, properties {soundWaveAssetPtr=&/Spot_Bamboo_Creak_9}}
]
properties {
 attenuationSettings=/Game/Sounds/Attenuation/Fort_TileAmbient_OneShot_Wider_Att
 bOverrideConcurrency=true
 maxDistance=8000
 soundClassObject=&Fort_Ambient_Outdoors_Athena
 volumeMultiplier=0.752
 attenuationOverrides {falloffDistance=2000, nonSpatializedRadiusStart=700, stereoSpread=1000, attenuationShapeExtents {x=925}}
 concurrencyOverrides {maxCount=1, voiceStealReleaseTime=2}
}
