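# Sound cue for a Japanese warbler bird call, an outdoor ambient spot sound for the Sakura Springs POI.
# A quality-level root sends its first two levels to a random pick of six waves and its last two to a smaller pick of three.
# Each pick passes through a modulator that varies pitch and volume.
# Bird one-shot attenuation and the outdoor ambient sound class shape playback at reduced volume.
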SoundCue /Game/Sounds/Asteria/POI/SakuraSprings/Spots/Spot_Sakura_Springs_Warbler_Bird_Cue
root=QualityLevel_0
nodes [
 {id=Modulator_0, children [Random_0], properties {pitchMax=1.1, pitchMin=0.9, volumeMax=1.2, volumeMin=0.5}}
 {id=Modulator_1, children [Random_1], properties {pitchMax=1.1, pitchMin=0.9, volumeMax=1.2, volumeMin=0.5}}
 {id=QualityLevel_0, children [Modulator_0,Modulator_0,Modulator_1,Modulator_1]}
 {id=Random_0, children [WavePlayer_0,WavePlayer_1,WavePlayer_2,WavePlayer_3,WavePlayer_4,WavePlayer_5], properties {weights [1,1,1,1,1,1]}}
 {id=Random_1, children [WavePlayer_6,WavePlayer_7,WavePlayer_8], properties {weights [1,1,1]}}
 {id=WavePlayer_0, properties {soundWaveAssetPtr=&/Spot_Bird_JapaneseWarbler_1}}
 {id=WavePlayer_1, properties {soundWaveAssetPtr=&/Spot_Bird_JapaneseWarbler_2}}
 {id=WavePlayer_2, properties {soundWaveAssetPtr=&/Spot_Bird_JapaneseWarbler_3}}
 {id=WavePlayer_3, properties {soundWaveAssetPtr=&/Spot_Bird_JapaneseWarbler_4}}
 {id=WavePlayer_4, properties {soundWaveAssetPtr=&/Spot_Bird_JapaneseWarbler_5}}
 {id=WavePlayer_5, properties {soundWaveAssetPtr=&/Spot_Bird_JapaneseWarbler_6}}
 {id=WavePlayer_6, properties {soundWaveAssetPtr=&Spot_Bird_JapaneseWarbler_1}}
 {id=WavePlayer_7, properties {soundWaveAssetPtr=&Spot_Bird_JapaneseWarbler_2}}
 {id=WavePlayer_8, properties {soundWaveAssetPtr=&Spot_Bird_JapaneseWarbler_3}}
]
properties {
 attenuationSettings=&Fort_TileAmbient_OneShot_Bird_Att
 maxDistance=8005
 soundClassObject=&Fort_Ambient_Outdoors_Athena
 volumeMultiplier=0.226
 preEffectBusSends [{}]
}
